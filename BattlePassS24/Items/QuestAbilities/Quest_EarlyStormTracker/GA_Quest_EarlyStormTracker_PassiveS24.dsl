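; Passive gameplay ability for the Season 24 Early Storm Tracker quest that reacts to storm safe zone phase changes.
; On activation it listens to the game state's stateful safe zone phase event through the GameplayEventRouter, which can also supply saved state.
; The phase index feeds StormZoneUpdated, which applies a configured GameplayEffect to the owner above phase zero and otherwise ends the ability.
; Ending the ability stops the stored listener task.
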(event Ability|EventFailedToActivatePassiveAbility
  (params (ActorInfo :type Gameplay%20Ability%20Actor%20Info%20Structure)))

(event Ability|EventActivateAbility
  (|CallParentFunction)
  (bind _returnvalue (GameplayEventRouter|GetEventRouter "/Script/FortniteGame.FortGameState" (Ability|GetAvatarActorfromActorInfo)))
  (bind (_asynctaskproxy _payload) (GameplayEventRouter|StartListeningToStatefulEvent _returnvalue "/Script/FortniteGame.FortSafeZonePhaseUpdatedEvent")
    (:then
      (Variables|Default|SetSafeZonePhaseUpdatedAsyncTask _asynctaskproxy))
    (:OnEventReceived
      _)
    (:OnSavedEventStateExists
      _))
  (bind _payload_safezonephaseindex (Utilities|Struct|BreakFortSafeZonePhaseUpdatedEvent _payload))
  (CallFunction|StormZoneUpdated _payload_safezonephaseindex))

(event Custom|Storm%20Zone%20Updated
  (params (SafeZonePhase :type Integer))
  (if (> SafeZonePhase 0)
    (Ability|ApplyGameplayEffectToOwner (Variables|Default|GetGE))
    (Ability|EndAbility)))

(event Ability|EventOnEndAbility
  (params (bWasCancelled :type Boolean))
  (|CallParentFunction bWasCancelled)
  (bind _safezonephaseupdatedasynctask (Variables|Default|GetSafeZonePhaseUpdatedAsyncTask)
    (:then
      (GameplayEventRouter|StopListeningtoStatefulEvent _safezonephaseupdatedasynctask))
    (:else)))
